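# Masked, Default Lit material in the creature backpack cosmetic folder, named as invisible glass and likely meant to hide a glass part of that backpack.
# Its graph holds a single Constant left at its default, with no textures, parameters or visible connections.
# It is flagged for use on skeletal meshes and with mesh deformers.
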Material /BRCosmetics/Accessories/FORT_Backpacks/M_MED_Creature_Backpack/Materials/M_invisibleGlass
graph {
 properties {
  bUsedWithMeshDeformer=true
  bUsedWithSkeletalMesh=true
  blendMode=BLEND_Masked
  pixelDepthOffsetMode=PDOM_Legacy
  usedShadingModels=Default Lit
 }
 nodes [Constant()]
}
