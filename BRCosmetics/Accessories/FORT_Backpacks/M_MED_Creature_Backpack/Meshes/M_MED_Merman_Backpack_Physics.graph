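# Physics asset for the Merman creature backpack cosmetic, defining capsule bodies and constraints for its root, two hoses and dangling slime bones.
# The hoses and five slime bodies each constrain directly to backpack_root, with a slime tail chained off slime_dyn_05 as the only second-level link.
# Slime joints share wide, uniform swing and twist limits, while the hoses have narrow, asymmetric swing limits and a zero twist limit.
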PhysicsAsset /BRCosmetics/Accessories/FORT_Backpacks/M_MED_Creature_Backpack/Meshes/M_MED_Merman_Backpack_Physics
bodies {
 backpack_root [{shapeType=Capsule, length=5, radius=5}]
 hose_l [{shapeType=Capsule, length=5.5834, radius=2.8833, center=<3.5623 0.3613 1.3664>, rotation=<rot 7.7322 -90 2.3588>}]
 hose_r [{shapeType=Capsule, length=5.6314, radius=2.9151, center=<-2.3221 -1.6948 1.9877>, rotation=<rot 15.7877 -90 -4.1583>}]
 slime_dyn_01 [{shapeType=Capsule, length=5, radius=3, center=<0.7627 -0.0587 -2.1673>}]
 slime_dyn_02 [{shapeType=Capsule, length=5.25, radius=3, center=<-0.4999 -0.7193 -1.2982>}]
 slime_dyn_03 [{shapeType=Capsule, length=4, radius=2.5, center=<-0.544 0 -1.8486>}]
 slime_dyn_04 [{shapeType=Capsule, length=5, radius=3.75, center=<0 -0.6823 -0.3557>}]
 slime_dyn_05 [{shapeType=Capsule, length=2.5, radius=3, center=<-0.4734 0 -2.9087>}]
 slime_dyn_05_tail [{shapeType=Capsule, length=9.75, radius=3.25, center=<0 0 -8.3914>}]
}
constraints [
 {bone1Name=slime_dyn_01, bone2Name=backpack_root, swing1LimitDegrees=45, swing2LimitDegrees=45, twistLimitDegrees=45}
 {bone1Name=slime_dyn_02, bone2Name=backpack_root, swing1LimitDegrees=45, swing2LimitDegrees=45, twistLimitDegrees=45}
 {bone1Name=slime_dyn_03, bone2Name=backpack_root, swing1LimitDegrees=45, swing2LimitDegrees=45, twistLimitDegrees=45}
 {bone1Name=slime_dyn_04, bone2Name=backpack_root, swing1LimitDegrees=45, swing2LimitDegrees=45, twistLimitDegrees=45}
 {bone1Name=hose_l, bone2Name=backpack_root, swing1LimitDegrees=20, swing2LimitDegrees=5, twistLimitDegrees=0}
 {bone1Name=hose_r, bone2Name=backpack_root, swing1LimitDegrees=20, swing2LimitDegrees=5, twistLimitDegrees=0}
 {bone1Name=slime_dyn_05, bone2Name=backpack_root, swing1LimitDegrees=45, swing2LimitDegrees=45, twistLimitDegrees=45}
 {bone1Name=slime_dyn_05_tail, bone2Name=slime_dyn_05, swing1LimitDegrees=45, swing2LimitDegrees=45, twistLimitDegrees=45}
]
