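# Animation Blueprint that adds secondary physics motion to the Merman creature backpack cosmetic, including its slime bones.
# The input pose passes through a partially blended RigidBody node, then an angular-limited Anim Dynamics tail bone.
# Five slime bones follow as low-alpha, damped linear springs with locked rotation, giving translational jiggle.
# All physics nodes carry an LOD threshold; the event graph does nothing, and a test sequence player is left unconnected.
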AnimBlueprint /BRCosmetics/Accessories/FORT_Backpacks/M_MED_Creature_Backpack/Meshes/M_MED_Merman_Backpack_AnimBP
target_skeleton=/Game/Accessories/FORT_Backpacks/M_Med_Creature_Backpack/Meshes/M_MED_Merman_Backpack_Skeleton
graphs [
 AnimGraph {nodes [
   AnimGraphNode_SubInput_0 = Input Pose()
   n1 = RigidBody(ComponentPose=Local To Component(LocalPose=@AnimGraphNode_SubInput_0), Alpha=0.35) {
    bOverrideWorldGravity=true
    lODThreshold=1
    overridePhysicsAsset=&/M_MED_Merman_Backpack_Physics
    alphaBoolBlend {blendOption=Linear}
    componentAppliedLinearAccClamp=<2000 2000 2000>
    componentLinearAccScale=<3 3 3>
    componentLinearVelScale=<8 8 8>
    overrideWorldGravity {z=-980}
   }
   n2 = Anim Dynamics - Bone: slime_dyn_05_tail(ComponentPose=@n1, Alpha=1, ExternalForce="0,0,0") {
    bGravityOverrideInSimSpace=true
    lODThreshold=1
    numSolverIterationsPostUpdate=4
    numSolverIterationsPreUpdate=16
    simulationSpace=Actor
    alphaBoolBlend {blendOption=Linear}
    boundBone {boneName=slime_dyn_05_tail}
    physicsBodyDefinitions [{
      sphereCollisionRadius=0
      boundBone {boneName=slime_dyn_05_tail}
      boxExtents=<10 10 30>
      constraintSetup {twistAxis=AxisZ, angularLimitsMax=<60 60 5>, angularLimitsMin=<-60 -60 -5>, angularTarget=<0 0 0>}
      localJointOffset=<0 0 -15>
     }]
   }
   n3 = Anim Dynamics - Bone: slime_dyn_01(ComponentPose=@n2, Alpha=0.2, ExternalForce="0,0,0") {
    bGravityOverrideInSimSpace=true
    bLinearSpring=true
    bOverrideLinearDamping=true
    gravityScale=0.2
    lODThreshold=1
    linearDampingOverride=0.99
    linearSpringConstant=320
    simulationSpace=Actor
    alphaBoolBlend {blendOption=Linear}
    boundBone {boneName=slime_dyn_01}
    physicsBodyDefinitions [{
      sphereCollisionRadius=0
      boundBone {boneName=slime_dyn_01}
      boxExtents=<30 25 15>
      constraintSetup {
       twistAxis=AxisZ
       angularLimitsMax=<0 0 0>
       angularLimitsMin=<0 0 0>
       angularTarget=<0 0 0>
       linearAxesMax=<10 10 10>
       linearAxesMin=<-10 -10 -10>
      }
      localJointOffset=<0 0 0>
     }]
   }
   n4 = Anim Dynamics - Bone: slime_dyn_02(ComponentPose=@n3, Alpha=0.15, ExternalForce="0,0,0") {
    bGravityOverrideInSimSpace=true
    bLinearSpring=true
    bOverrideLinearDamping=true
    gravityScale=0.2
    lODThreshold=1
    linearDampingOverride=0.95
    linearSpringConstant=380
    simulationSpace=Actor
    alphaBoolBlend {blendOption=Linear}
    boundBone {boneName=slime_dyn_02}
    physicsBodyDefinitions [{
      sphereCollisionRadius=0
      boundBone {boneName=slime_dyn_02}
      boxExtents=<16 12 32>
      constraintSetup {
       twistAxis=AxisZ
       angularLimitsMax=<0 0 0>
       angularLimitsMin=<0 0 0>
       angularTarget=<0 0 0>
       linearAxesMax=<10 10 10>
       linearAxesMin=<-10 -10 -10>
      }
      localJointOffset=<0 0 0>
     }]
   }
   n5 = Anim Dynamics - Bone: slime_dyn_03(ComponentPose=@n4, Alpha=0.35, ExternalForce="0,0,0") {
    bGravityOverrideInSimSpace=true
    bLinearSpring=true
    bOverrideLinearDamping=true
    gravityScale=0.2
    lODThreshold=1
    linearDampingOverride=0.98
    linearSpringConstant=300
    simulationSpace=Actor
    alphaBoolBlend {blendOption=Linear}
    boundBone {boneName=slime_dyn_03}
    physicsBodyDefinitions [{
      sphereCollisionRadius=0
      boundBone {boneName=slime_dyn_03}
      boxExtents=<10 10 30>
      constraintSetup {
       twistAxis=AxisZ
       angularLimitsMax=<0 0 0>
       angularLimitsMin=<0 0 0>
       angularTarget=<0 0 0>
       linearAxesMax=<10 10 10>
       linearAxesMin=<-10 -10 -10>
      }
      localJointOffset=<0 0 0>
     }]
   }
   n6 = Anim Dynamics - Bone: slime_dyn_04(ComponentPose=@n5, Alpha=0.22, ExternalForce="0,0,0") {
    bGravityOverrideInSimSpace=true
    bLinearSpring=true
    bOverrideLinearDamping=true
    gravityScale=0.2
    lODThreshold=1
    linearDampingOverride=0.95
    linearSpringConstant=400
    simulationSpace=Actor
    alphaBoolBlend {blendOption=Linear}
    boundBone {boneName=slime_dyn_04}
    physicsBodyDefinitions [{
      sphereCollisionRadius=0
      boundBone {boneName=slime_dyn_04}
      boxExtents=<18 18 29>
      constraintSetup {
       twistAxis=AxisZ
       angularLimitsMax=<0 0 0>
       angularLimitsMin=<0 0 0>
       angularTarget=<0 0 0>
       linearAxesMax=<10 10 10>
       linearAxesMin=<-10 -10 -10>
      }
      localJointOffset=<0 0 0>
     }]
   }
   n7 = Anim Dynamics - Bone: slime_dyn_05(ComponentPose=@n6, Alpha=0.4, ExternalForce="0,0,0") {
    bGravityOverrideInSimSpace=true
    bLinearSpring=true
    bOverrideLinearDamping=true
    gravityScale=0.2
    lODThreshold=1
    linearDampingOverride=0.96
    linearSpringConstant=310
    simulationSpace=Actor
    alphaBoolBlend {blendOption=Linear}
    boundBone {boneName=slime_dyn_05}
    physicsBodyDefinitions [{
      sphereCollisionRadius=0
      boundBone {boneName=slime_dyn_05}
      boxExtents=<15 15 25>
      constraintSetup {
       twistAxis=AxisZ
       angularLimitsMax=<0 0 0>
       angularLimitsMin=<0 0 0>
       angularTarget=<0 0 0>
       linearAxesMax=<10 10 10>
       linearAxesMin=<-10 -10 -10>
      }
      localJointOffset=<0 0 0>
     }]
   }
   Output Pose(Result=Component To Local(ComponentPose=@n7)) {layerGroup=None}
   Sequence Player 'Backpack_PhysicsTest_Animation'() {sequence=/Game/Accessories/FORT_Backpacks/TestAnimations/Backpack_PhysicsTest_Animation}
  ]}
 EventGraph {nodes [
   TryGetPawnOwner()
   Event BlueprintUpdateAnimation()
  ]}
]
variables []
